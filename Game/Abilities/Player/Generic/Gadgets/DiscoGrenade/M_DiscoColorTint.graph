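# Additive unlit material for mesh particles that creates a moving multicolor disco tint for the player's disco grenade gadget.
# Two tiled samples of one noise texture pan horizontally in opposite directions, with three channels tinted by DiscoColor parameters and summed.
# The color sum feeds an emissive adjustment function and a powered mask branch.
# That branch combines particle alpha, Fresnel, depth fade and a V-axis gradient that fades both UV edges.
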Material /Game/Abilities/Player/Generic/Gadgets/DiscoGrenade/M_DiscoColorTint
graph {
 properties {
  bUsedWithMeshParticles=true
  bUsedWithNiagaraMeshParticles=true
  blendMode=BLEND_Additive
  pixelDepthOffsetMode=PDOM_Legacy
  shadingModel=MSM_Unlit
  usedShadingModels=Unlit
 }
 nodes [
  n1 = TextureSample(UVs=Panner(Coordinate=TextureCoordinate() {uTiling=8, vTiling=4}) {speedX=0.3}) {texture=&T_NoisePack_01}
  n2 = VectorParameter() {parameterName=DiscoColor2, defaultValue=<rgba 0.01 0.0768 1 0>}
  n3 = TextureSample(UVs=Panner(Coordinate=TextureCoordinate() {uTiling=8, vTiling=4}) {speedX=-0.25}) {texture=&T_NoisePack_01}
  n4 = VectorParameter() {parameterName=DiscoColor3, defaultValue=<rgba 0.5 0.4218 0.0605 0>}
  n5 = Add(
   A=Multiply(A=@n1.R, B=VectorParameter() {parameterName=DiscoColor1, defaultValue=<rgba 0.5 0.055 0.463 0>}.RGB)
   B=Add(A=Multiply(A=@n1.B, B=@n2.RGB), B=Multiply(A=@n3.G, B=@n4.RGB))
  )
  MaterialFunctionCall(Emissive=@n5) {materialFunction=&MF_FX_EmissiveAdjustments}
  n6 = Multiply(
   A=ParticleColor().A
   B=MaterialFunctionCall(Exponent=Constant() {r=3}) {materialFunction=&MF_Fresnel_FX}.Fresnel
  )
  n7 = Constant() {r=16}
  n8 = ComponentMask(None=TextureCoordinate()) {g=true}
  n9 = Multiply(A=Multiply(A=@n8, B=OneMinus(None=@n8))) {constB=4}
  Multiply(A=Power(Base=@n5), B=Multiply(A=DepthFade(Opacity=@n6, FadeDistance=@n7), B=Power(Base=@n9) {constExponent=3}))
 ]
}
