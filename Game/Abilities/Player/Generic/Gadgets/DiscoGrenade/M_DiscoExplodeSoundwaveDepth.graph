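# Translucent, unlit, two-sided particle mesh material for the soundwave in the Disco Grenade explosion.
# Particle color gets a time-driven hue shift, which a parameter can blend out, then passes through emissive adjustments.
# Particle alpha, a scene-depth distance mask and Fresnel multiply into an opacity-style mask.
# Two noise samples, panned oppositely along U with V tiling zeroed, multiply to drive an in-and-out displacement along vertex normals.
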Material /Game/Abilities/Player/Generic/Gadgets/DiscoGrenade/M_DiscoExplodeSoundwaveDepth
graph {
 properties {
  bUsedWithMeshParticles=true
  bUsedWithNiagaraMeshParticles=true
  blendMode=BLEND_Translucent
  pixelDepthOffsetMode=PDOM_Legacy
  shadingModel=MSM_Unlit
  twoSided=true
  usedShadingModels=Unlit
 }
 nodes [
  n1 = Frac(None=Multiply(A=Time(), B=ScalarParameter() {defaultValue=3, parameterName=HueShiftSpeed}) {constB=1.2})
  n2 = ParticleColor()
  MaterialFunctionCall(Emissive=LinearInterpolate(
    A=MaterialFunctionCall(Hue Shift Percentage=@n1, Texture=@n2.RGB) {materialFunction=&HueShift}.Result
    B=@n2.RGB
    Alpha=ScalarParameter() {parameterName=Disable Hue Shift}
   )) {materialFunction=&MF_FX_EmissiveAdjustments}
  n3 = ScalarParameter() {defaultValue=100, parameterName=DistanceDivisor}
  Multiply(
   A=Multiply(
    A=@n2.A
    B=MaterialFunctionCall!(Distance Bias, Distance Divisor=@n3, Scene Depth, Pixel Depth) {materialFunction=/Game/Effects/Fort_Effects/Materials/_Functions/MF_DistanceBehindTranslucency_01}.Normalized Inverted
   )
   B=MaterialFunctionCall() {materialFunction=&MF_Fresnel_FX}.Fresnel
  )
  n4 = TextureCoordinate() {vTiling=0}
  n5 = TextureCoordinate() {uTiling=2, vTiling=0}
  n6 = Multiply(
   A=TextureSample(UVs=Panner(Coordinate=@n4) {speedX=1}) {texture=&T_NoisePack_01}.R
   B=TextureSample(UVs=Panner(Coordinate=@n5) {speedX=-0.53}) {texture=&T_NoisePack_01}.R
  )
  Multiply(A=LinearInterpolate(Alpha=Multiply(A=@n6) {constB=2}) {constA=-5, constB=5}, B=VertexNormalWS())
 ]
}
