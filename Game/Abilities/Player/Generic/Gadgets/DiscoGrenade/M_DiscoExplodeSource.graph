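# Unlit translucent particle material for the disco grenade, apparently the glowing source of its explosion, usable on Niagara and particle sprites.
# Particle color RGB runs through the shared FX emissive adjustment function.
# Vertically panning noise modulates the radius of a soft, UV-centered sphere mask, giving a wobbling circular edge without a shape texture.
# That mask is multiplied by particle alpha and depth fade to soften intersections with geometry.
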Material /Game/Abilities/Player/Generic/Gadgets/DiscoGrenade/M_DiscoExplodeSource
graph {
 properties {
  bUsedWithNiagaraSprites=true
  bUsedWithParticleSprites=true
  blendMode=BLEND_Translucent
  pixelDepthOffsetMode=PDOM_Legacy
  shadingModel=MSM_Unlit
  usedShadingModels=Unlit
 }
 nodes [
  n1 = ParticleColor()
  MaterialFunctionCall(Emissive=@n1.RGB) {materialFunction=&MF_FX_EmissiveAdjustments}
  n2 = Panner() {speedY=1}
  n3 = SphereMask(
   A=TextureCoordinate()
   B=Constant() {r=0.5}
   Radius=LinearInterpolate(Alpha=TextureSample(UVs=@n2) {texture=&T_NoisePack_01}.R) {constA=0.4, constB=0.5}
   Hardness=Constant() {r=0.3}
  ) {attenuationRadius=0.5}
  Multiply(A=Multiply(A=@n1.A, B=Power(Base=@n3)), B=DepthFade(FadeDistance=Constant() {r=8}))
 ]
}
