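# Niagara effect for the disco grenade gadget: a glowing disco ball hovering above the origin with pulsing party lights.
# A rising flare with a light opens it as a persistent ball mesh appears, flashing through brightness curves and animated material parameters.
# Coronas, rotating beam meshes and a color-shifting light repeat on a shared 0.46-second beat, alongside continuous sparkles.
# DiscoBallLoop is disabled, some modules come from Cascade conversion, and no user parameters are exposed.
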NiagaraSystem /Game/Abilities/Player/Generic/Gadgets/DiscoGrenade/NS_DiscoBall
dependencies {
 usedDataInterfaces [&NiagaraDataInterfaceCurve, &NiagaraDataInterfaceVectorCurve]
 usedDynamicInputs []
 usedModules [&SystemState]
}
emitters {
 Corona {
  emitterUpdateScript [
   &EmitterState(
    Delay First Loop Only=true
    Inactive Response=::NewEnumerator1
    Life Cycle Mode=::NewEnumerator1
    Loop Delay=0.46
    Loop Duration=0.46
    UseLoopDelay=true
   )
   &SpawnBurst_Instantaneous(Use Spawn Probability=false, UseLoopCountLimit=false)
  ]
  particleSpawnScript [
   &InitializeParticle(Lifetime=0.55, Sprite Size=<55 55>, Sprite Size Mode=::NewEnumerator4, UsePositionOffset=false, Write Material Random=true)
   &Color(Color=&MakeLinearColorFromVectorAndFloat("Vector (RGB)"=<1.7183 4.9503 5>, "Float (Alpha)"=1))
   &DynamicMaterialParameters(
    Index 0 Param 3=&RandomRangeFloat(Minimum=0.5, Maximum=0.7, Evaluation Type=::NewEnumerator1)
    Module.Param0WriteEnabled=false
    Module.Param1WriteEnabled=false
    Module.Param2WriteEnabled=true
    Module.Param3WriteEnabled=false
   ) {module=DynamicMaterialParameters001}
   /CascadeToNiagaraConverter/NiagaraScripts/ModuleScripts/AddVectorToPosition(Vector:Vector3f=&TransformVector(Vector=<0 0 230>, Destination Space=::Simulation))
  ]
  particleUpdateScript [
   &ParticleState()
   &CameraOffset(Camera Offset Amount=50)
   &DynamicMaterialParameters(
    Index 0 Param 1=&FloatFromCurve(FloatCurve=DataInterface() {propertyValues {Curve {keys [
         {arriveTangent=-2.2498, interpMode=RCIM_Cubic, leaveTangent=-2.2498, tangentMode=RCTM_User, time=0, value=0.8}
         {arriveTangent=0.2, interpMode=RCIM_Cubic, leaveTangent=0.2, time=0.2, value=0.65}
         <at 1 1>
        ]}}})
    Index 0 Param 2=1
    Index 0 Param 4=1
    Module.Param0WriteEnabled=true
    Module.Param1WriteEnabled=true
    Module.Param2WriteEnabled=false
    Module.Param3WriteEnabled=true
   )
   &ScaleColor(
    Scale Alpha=&FloatFromCurve(FloatCurve=DataInterface() {propertyValues {Curve {keys [
         {arriveTangent=22.3315, interpMode=RCIM_Cubic, leaveTangent=22.3315, tangentMode=RCTM_User, time=0, value=0}
         <at 0.1 1 RCIM_Cubic>
         <at 1 0>
        ]}}})
    Scale RGB=&VectorFromCurve(VectorCurve=DataInterface() {propertyValues {
       XCurve {keys [<at 0 1 RCIM_Cubic>,<at 1 0.3728>]}
       YCurve {keys [<at 0 0.6851 RCIM_Cubic>,<at 1 1>]}
       ZCurve {keys [<at 0 0.8647 RCIM_Cubic>,<at 1 0.6644>]}
      }})
    ScaleA=true
    ScaleRGB=true
   )
  ]
  rendererClasses [&NiagaraSpriteRendererProperties]
  renderers [{rendererClass=&NiagaraSpriteRendererProperties}]
 }
 CoronaSpread {
  emitterUpdateScript [
   &EmitterState(
    Delay First Loop Only=true
    Inactive Response=::NewEnumerator1
    Life Cycle Mode=::NewEnumerator1
    Loop Delay=0.45
    Loop Duration=0.46
    UseLoopDelay=true
   )
   &SpawnBurst_Instantaneous(Use Spawn Probability=false, UseLoopCountLimit=false)
  ]
  particleSpawnScript [
   &InitializeParticle(Lifetime=0.7, Sprite Size=<64 64>, Sprite Size Mode=::NewEnumerator4, UsePositionOffset=false, Write Material Random=true)
   &Color(Color=&MakeLinearColorFromVectorAndFloat("Vector (RGB)"=<1.9022 0.7396 2>, "Float (Alpha)"=0.8))
   &DynamicMaterialParameters(
    Index 0 Param 3=&RandomRangeFloat(Minimum=0.5, Maximum=0.7, Evaluation Type=::NewEnumerator1)
    Module.Param0WriteEnabled=false
    Module.Param1WriteEnabled=false
    Module.Param2WriteEnabled=true
    Module.Param3WriteEnabled=false
   ) {module=DynamicMaterialParameters001}
   &AddVectorToPosition(Vector:Vector3f=&TransformVector(Vector=<0 0 230>, Destination Space=::Simulation))
  ]
  particleUpdateScript [
   &ParticleState()
   &ScaleSpriteSize(
    Scale Sprite Size Mode=::NewEnumerator0
    Uniform Scale Factor=&FloatFromCurve(FloatCurve=DataInterface() {propertyValues {Curve {keys [
         {arriveTangent=2.8429, interpMode=RCIM_Cubic, leaveTangent=2.8429, tangentMode=RCTM_User, time=0, value=0.7}
         {arriveTangent=0.1481, interpMode=RCIM_Cubic, leaveTangent=0.1481, tangentMode=RCTM_User, time=1, value=1.5}
        ]}}})
   )
   &CameraOffset(Camera Offset Amount=50)
   &DynamicMaterialParameters(
    Index 0 Param 1=&FloatFromCurve(FloatCurve=DataInterface() {propertyValues {Curve {keys [
         {arriveTangent=-3.7348, interpMode=RCIM_Cubic, leaveTangent=-3.7348, tangentMode=RCTM_User, time=0, value=1.5}
         {arriveTangent=-0.6045, interpMode=RCIM_Cubic, leaveTangent=-0.6045, tangentMode=RCTM_User, time=1, value=0.2}
        ]}}})
    Index 0 Param 2=1
    Index 0 Param 4=1
    Module.Param0WriteEnabled=true
    Module.Param1WriteEnabled=true
    Module.Param2WriteEnabled=false
    Module.Param3WriteEnabled=true
   )
   &ScaleColor(
    Scale Alpha=&FloatFromCurve(FloatCurve=DataInterface() {propertyValues {Curve {keys [
         {arriveTangent=24.9021, interpMode=RCIM_Cubic, leaveTangent=24.9021, tangentMode=RCTM_User, time=0, value=0}
         {arriveTangent=-2.1827, interpMode=RCIM_Cubic, leaveTangent=-2.1827, tangentMode=RCTM_User, time=0.1, value=1}
         <at 1 0>
        ]}}})
    Scale RGB=&VectorFromCurve(VectorCurve=DataInterface() {propertyValues {
       XCurve {keys [<at 0 1.5 RCIM_Cubic>,<at 1 1>]}
       YCurve {keys [<at 0 1.0276 RCIM_Cubic>,<at 1 1>]}
       ZCurve {keys [<at 0 1.297 RCIM_Cubic>,<at 1 1>]}
      }})
    ScaleA=true
    ScaleRGB=true
   )
  ]
  rendererClasses [&NiagaraSpriteRendererProperties]
  renderers [{rendererClass=&NiagaraSpriteRendererProperties}]
 }
 DiscoBallLoop {
  bEnabled=false
  emitterUpdateScript [
   &EmitterState()
   &SpawnBurst_Instantaneous(Spawn Time=1.84, Use Spawn Probability=false, UseLoopCountLimit=true)
  ]
  particleSpawnScript [
   &InitializeParticle(Lifetime=1.84, Mesh Scale=<0.625 0.625 0.625>, Mesh Scale Mode=::NewEnumerator4, UsePositionOffset=false, Write Material Random=true)
   &Color(Color=&MakeLinearColorFromVectorAndFloat("Vector (RGB)"=<2.9151 2.2163 3>, "Float (Alpha)"=1))
   &AddVectorToPosition(Vector:Vector3f=&TransformVector(Vector=<0 0 230>, Destination Space=::Simulation))
   &AddVectorToPosition(Vector:Vector3f=&TransformVector(Vector=<0 0 33>, Destination Space=::Simulation)) {module=AddVectorToPosition001}
  ]
  particleUpdateScript [
   &ParticleState(Kill Particles When Lifetime Has Elapsed=false, Let Infinitely Lived Particles Die When Emitter Deactivates=true)
   &ScaleColor(
    Scale RGB=&VectorFromFloat(Value=&FloatFromCurve(FloatCurve=DataInterface() {propertyValues {Curve {keys [
          {arriveTangent=-38.4417, interpMode=RCIM_Cubic, leaveTangent=-38.4417, tangentMode=RCTM_User, time=0, value=5}
          <at 0.245 0.2 RCIM_Cubic>
          {arriveTangent=-54.5918, interpMode=RCIM_Cubic, leaveTangent=-54.5918, tangentMode=RCTM_User, time=0.25, value=5}
          <at 0.495 0.2 RCIM_Cubic>
          {arriveTangent=-56.3829, interpMode=RCIM_Cubic, leaveTangent=-56.3829, tangentMode=RCTM_User, time=0.5, value=5}
          <at 0.745 0.2 RCIM_Cubic>
          {arriveTangent=-54.4251, interpMode=RCIM_Cubic, leaveTangent=-54.4251, tangentMode=RCTM_User, time=0.75, value=5}
          <at 1 0.2 RCIM_Cubic>
         ]}}}))
    ScaleA=true
    ScaleRGB=true
   )
   &DynamicMaterialParameters(
    Index 0 Param 2=&FloatFromCurve(FloatCurve=DataInterface() {propertyValues {bOptimizeLUT=false, bUseLUT=false, Curve {postInfinityExtrap=RCCE_Linear, keys [{arriveTangent=0.5, time=0, value=0}, <at 1 0.5>]}}})
    Index 0 Param 3=1
    Index 0 Param 4=1
    Module.Param0WriteEnabled=true
    Module.Param1WriteEnabled=true
    Module.Param2WriteEnabled=true
    Module.Param3WriteEnabled=true
   )
  ]
  rendererClasses [&NiagaraMeshRendererProperties]
  renderers [{rendererClass=&NiagaraMeshRendererProperties, rendererIndex=0}]
 }
 DiscoBallSpawn {
  emitterUpdateScript [
   &EmitterState()
   &SpawnBurst_Instantaneous(Use Spawn Probability=false, UseLoopCountLimit=true)
  ]
  particleSpawnScript [
   &InitializeParticle(Lifetime=1.84, Mesh Scale=<0.625 0.625 0.625>, Mesh Scale Mode=::NewEnumerator4, UsePositionOffset=false, Write Material Random=true)
   &Color(Color=&MakeLinearColorFromVectorAndFloat("Vector (RGB)"=<2.9151 2.2163 3>, "Float (Alpha)"=1))
   &AddVectorToPosition(Vector:Vector3f=&TransformVector(Vector=<0 0 230>, Destination Space=::Simulation))
   &AddVectorToPosition(Vector:Vector3f=&TransformVector(Vector=<0 0 33>, Destination Space=::Simulation)) {module=AddVectorToPosition001}
  ]
  particleUpdateScript [
   &ParticleState(Kill Particles When Lifetime Has Elapsed=false, Let Infinitely Lived Particles Die When Emitter Deactivates=true)
   &ScaleColor(
    Scale RGB=&VectorFromFloat(Value=&FloatFromCurve(FloatCurve=DataInterface() {propertyValues {Curve {keys [
          {arriveTangent=-38.4417, interpMode=RCIM_Cubic, leaveTangent=-38.4417, tangentMode=RCTM_User, time=0, value=25}
          {arriveTangent=-54.5918, interpMode=RCIM_Cubic, leaveTangent=-54.5918, tangentMode=RCTM_User, time=0.25, value=5}
          <at 0.495 0.2 RCIM_Cubic>
          {arriveTangent=-56.3829, interpMode=RCIM_Cubic, leaveTangent=-56.3829, tangentMode=RCTM_User, time=0.5, value=5}
          <at 0.745 0.2 RCIM_Cubic>
          {arriveTangent=-54.4251, interpMode=RCIM_Cubic, leaveTangent=-54.4251, tangentMode=RCTM_User, time=0.75, value=5}
          <at 1 0.2 RCIM_Cubic>
         ]}}}))
    ScaleA=true
    ScaleRGB=true
   )
   &DynamicMaterialParameters(
    Index 0 Param 1=&FloatFromCurve(FloatCurve=DataInterface() {propertyValues {Curve {keys [{arriveTangent=-1, time=0, value=1}, <at 0.5 0>]}}})
    Index 0 Param 2=&FloatFromCurve(FloatCurve=DataInterface() {propertyValues {
       bUseLUT=false
       Curve {
        postInfinityExtrap=RCCE_Linear
        keys [
         <at 0.2 0 RCIM_Cubic>
         {arriveTangent=1.7244, interpMode=RCIM_Cubic, leaveTangent=1.7244, tangentMode=RCTM_User, time=0.3, value=-0.1}
         {arriveTangent=0.4445, interpMode=RCIM_Cubic, leaveTangent=0.4445, tangentMode=RCTM_User, time=1, value=0.5}
        ]
       }
      }})
    Index 0 Param 3=1
    Index 0 Param 4=1
    Module.Param0WriteEnabled=true
    Module.Param1WriteEnabled=true
    Module.Param2WriteEnabled=true
    Module.Param3WriteEnabled=true
   )
  ]
  rendererClasses [&NiagaraMeshRendererProperties]
  renderers [{rendererClass=&NiagaraMeshRendererProperties, rendererIndex=0}]
 }
 FlareStart {
  emitterUpdateScript [
   &EmitterState(Inactive Response=::NewEnumerator1, Life Cycle Mode=::NewEnumerator1, Loop Behavior=::NewEnumerator1, Loop Duration=1, UseLoopDelay=true)
   &SpawnBurst_Instantaneous(Use Spawn Probability=false, UseLoopCountLimit=false)
  ]
  particleSpawnScript [
   &InitializeParticle(Lifetime=0.2, Sprite Size=<100 100>, Sprite Size Mode=::NewEnumerator4, UsePositionOffset=false, Write Material Random=true)
   &Color(Color=&MakeLinearColorFromVectorAndFloat("Vector (RGB)"=<1.5195 0.3572 2>, "Float (Alpha)"=1))
  ]
  particleUpdateScript [
   /SetVariables(Particles.Position=&TransformPosition(
     Position=&ConvertVectorToPosition(
      Input Position=&VectorFromCurve(VectorCurve=DataInterface() {propertyValues {
         XCurve {keys [{interpMode=RCIM_Cubic, tangentMode=RCTM_User, time=0, value=0}, <at 1 0 RCIM_Cubic>]}
         YCurve {keys [{interpMode=RCIM_Cubic, tangentMode=RCTM_User, time=0, value=0}, <at 1 0 RCIM_Cubic>]}
         ZCurve {keys [{arriveTangent=291.1376, interpMode=RCIM_Cubic, leaveTangent=291.1376, tangentMode=RCTM_User, time=0, value=110}, <at 1 270 RCIM_Cubic>]}
        }})
      Conversion Operation:ENiagara_LWCConvertVecToPos=::NewEnumerator0
     )
     Source Space=::Local
     Destination Space=::Simulation
    )) {isSetParameters=true}
   /SetVariables(Particles.LightExponent=8) {isSetParameters=true}
   &ParticleState()
   &ScaleSpriteSize(
    Scale Sprite Size Mode=::NewEnumerator0
    Uniform Scale Factor=&FloatFromCurve(FloatCurve=DataInterface() {propertyValues {Curve {keys [
         {arriveTangent=4.0443, interpMode=RCIM_Cubic, leaveTangent=4.0443, tangentMode=RCTM_User, time=0, value=0.5}
         {arriveTangent=0.5, interpMode=RCIM_Cubic, leaveTangent=0.5, time=0.5, value=2}
         <at 1 1>
        ]}}})
   )
   &CameraOffset(Camera Offset Amount=50)
   &ScaleColor(
    Scale Alpha=&FloatFromCurve(FloatCurve=DataInterface() {propertyValues {Curve {keys [
         {arriveTangent=22.3315, interpMode=RCIM_Cubic, leaveTangent=22.3315, tangentMode=RCTM_User, time=0, value=0}
         <at 0.1 1 RCIM_Cubic>
         <at 1 0>
        ]}}})
    Scale RGB=&VectorFromCurve(VectorCurve=DataInterface() {propertyValues {
       XCurve {keys [<at 0 1 RCIM_Cubic>,<at 1 0.3728>]}
       YCurve {keys [<at 0 0.6851 RCIM_Cubic>,<at 1 1>]}
       ZCurve {keys [<at 0 0.8647 RCIM_Cubic>,<at 1 0.6644>]}
      }})
    ScaleA=true
    ScaleRGB=true
   )
   &CascadeConversion_LightProperties(LightBrightnessScale=10, LightColorScale=<1 1 1>, LightRadiusScale=3)
  ]
  rendererClasses [&NiagaraSpriteRendererProperties, &NiagaraLightRendererProperties]
  renderers [
   {rendererClass=&NiagaraSpriteRendererProperties}
   {rendererClass=&NiagaraLightRendererProperties, rendererIndex=1}
  ]
 }
 Light {
  emitterUpdateScript [
   &EmitterState(Inactive Response=::NewEnumerator1, Life Cycle Mode=::NewEnumerator1, Loop Duration=0.46, UseLoopDelay=true)
   &SpawnBurst_Instantaneous(Use Spawn Probability=false, UseLoopCountLimit=false)
  ]
  particleSpawnScript [
   &InitializeParticle(Lifetime=0.55, Sprite Size=<500 500>, Sprite Size Mode=::NewEnumerator4, UsePositionOffset=false, Write Material Random=true)
   &Color(Color=&MakeLinearColorFromVectorAndFloat("Vector (RGB)"=<1.7183 4.9503 5>, "Float (Alpha)"=1))
   &AddVectorToPosition(Vector:Vector3f=&TransformVector(Vector=<0 0 230>, Destination Space=::Simulation))
  ]
  particleUpdateScript [
   /SetVariables(Particles.LightExponent=8) {isSetParameters=true}
   &ParticleState()
   &ScaleColor(
    Scale Alpha=&FloatFromCurve(FloatCurve=DataInterface() {propertyValues {Curve {keys [<at 0 0 RCIM_Cubic>,<at 0.1 1 RCIM_Cubic>,<at 1 0 RCIM_Cubic>]}}})
    Scale RGB=&VectorFromCurve(VectorCurve=DataInterface() {propertyValues {
       XCurve {keys [<at 0 1 RCIM_Cubic>, {arriveTangent=-0.95, interpMode=RCIM_Cubic, leaveTangent=-0.95, time=0.5, value=0.05}, <at 1 0.05 RCIM_Cubic>]}
       YCurve {keys [<at 0 0.05 RCIM_Cubic>, {arriveTangent=0.95, interpMode=RCIM_Cubic, leaveTangent=0.95, time=0.5, value=0.6122}, <at 1 1 RCIM_Cubic>]}
       ZCurve {keys [<at 0 0.77 RCIM_Cubic>, {arriveTangent=-0.4762, interpMode=RCIM_Cubic, leaveTangent=-0.4762, time=0.5, value=1}, <at 1 0.2939 RCIM_Cubic>]}
      }})
    ScaleA=true
    ScaleRGB=true
   )
   &CascadeConversion_LightProperties(LightBrightnessScale=5, LightColorScale=<1 1 1>, LightRadiusScale=1)
  ]
  rendererClasses [&NiagaraLightRendererProperties]
  renderers [{rendererClass=&NiagaraLightRendererProperties, rendererIndex=0}]
 }
 SparkleStationary {
  emitterUpdateScript [
   &EmitterState()
   &SpawnBurst_Instantaneous(
    Spawn Count=&RandomRangeInt(Minimum=2, Maximum=5)
    Use Spawn Probability=false
    UseLoopCountLimit=false
   )
   &SpawnRate(SpawnRate=12, Use Spawn Probability=false)
  ]
  particleSpawnScript [
   &InitializeParticle(
    Lifetime=&RandomRangeFloat(Minimum=0.15, Maximum=0.25, Evaluation Type=::NewEnumerator1)
    Sprite Size=&Vector2DFromFloat(Value=&RandomRangeFloat(Minimum=8, Maximum=20, Evaluation Type=::NewEnumerator1))
    Sprite Size Mode=::NewEnumerator4
    UsePositionOffset=false
    Write Material Random=true
   )
   &ShapeLocation(MaskSpawnGroup=false, Override Seed=false, Sphere Radius=20, Sphere Surface Distribution=1, UseOwnerScale=true)
   &Color(Color=&MakeLinearColorFromVectorAndFloat(
     "Vector (RGB)"=&RandomRangeVector(Minimum=<8 8 8>, Maximum=<16 16 16>, Evaluation Type=::NewEnumerator1)
     "Float (Alpha)"=1
    ))
   &AddVectorToPosition(Vector:Vector3f=&TransformVector(Vector=<0 0 230>, Destination Space=::Simulation))
  ]
  particleUpdateScript [
   &ParticleState()
   &ScaleSpriteSize(
    Scale Sprite Size Mode=::NewEnumerator0
    Uniform Scale Factor=&FloatFromCurve(FloatCurve=DataInterface() {propertyValues {Curve {keys [<at 0 0 RCIM_Cubic>,<at 0.3 1 RCIM_Cubic>,<at 1 0>]}}})
   )
   &Drag(Drag=5, Use Linear Drag=true, Use Rotational Drag=true)
   &SolveForcesAndVelocity(Clamp Velocity=false, Limit Acceleration=false)
   &ScaleColor(
    Scale RGB=&VectorFromCurve(VectorCurve=DataInterface() {propertyValues {XCurve {keys [<at 0 1>,<at 0.5 1>,<at 1 0>]}, YCurve {keys [<at 0 0>,<at 0.5 1>,<at 1 0>]}, ZCurve {keys [<at 0 1>,<at 0.5 1>,<at 1 0>]}}})
    ScaleA=true
    ScaleRGB=true
   )
   &CameraOffset(Camera Offset Amount=60)
  ]
  rendererClasses [&NiagaraSpriteRendererProperties]
  renderers [{rendererClass=&NiagaraSpriteRendererProperties}]
 }
 SplineBeams {
  emitterUpdateScript [
   &EmitterState(
    Delay First Loop Only=true
    Inactive Response=::NewEnumerator1
    Life Cycle Mode=::NewEnumerator1
    Loop Delay=0.45
    Loop Duration=0.46
    UseLoopDelay=true
   )
   &SpawnBurst_Instantaneous(Spawn Count=2, Use Spawn Probability=false, UseLoopCountLimit=false)
  ]
  particleSpawnScript [
   /SetVariables(Particles.MeshRotationEuler:Vector3f=&RandomRangeVector(Minimum=<0 0 0>, Maximum=<1 1 1>, Evaluation Type=::NewEnumerator1)) {isSetParameters=true}
   /SetVariables(Particles.MeshOrientation=/CascadeToNiagaraConverter/NiagaraScripts/DynamicInputScripts/MakeQuatFromEuler(Euler:Vector3f=$Particles.MeshRotationEuler, Scale:NiagaraFloat=360)) {isSetParameters=true}
   /SetVariables(Particles.MeshRotationRateBase:Vector3f=&RandomRangeVector(Minimum=<-0.05 -0.05 -0.35>, Maximum=<0.05 0.05 -0.25>, Evaluation Type=::NewEnumerator1)) {isSetParameters=true}
   &InitializeParticle(
    Lifetime=0.55
    Mesh Scale=&VectorFromFloat(Value=&RandomRangeFloat(Minimum=1, Maximum=1.7, Evaluation Type=::NewEnumerator1))
    Mesh Scale Mode=::NewEnumerator4
    UsePositionOffset=false
    Write Material Random=true
   )
   &Color(Color=&MakeLinearColorFromVectorAndFloat("Vector (RGB)"=<3 3 3>, "Float (Alpha)"=0.2))
   &CameraOffset(Camera Offset Amount=&RandomRangeFloat(Minimum=5, Maximum=20, Evaluation Type=::NewEnumerator1))
   &DynamicMaterialParameters(
    Index 0 Param 1=&RandomRangeFloat(Minimum=1, Maximum=2, Evaluation Type=::NewEnumerator1)
    Index 0 Param 2=&RandomRangeFloat(Minimum=0, Maximum=1, Evaluation Type=::NewEnumerator1)
    Index 0 Param 3=&RandomRangeFloat(Minimum=2, Maximum=3, Evaluation Type=::NewEnumerator1)
    Module.Param0WriteEnabled=true
    Module.Param1WriteEnabled=true
    Module.Param2WriteEnabled=true
    Module.Param3WriteEnabled=false
   )
   &AddVectorToPosition(Vector:Vector3f=&TransformVector(Vector=<0 0 230>, Destination Space=::Simulation))
  ]
  particleUpdateScript [
   /SetVariables(Particles.MeshRotationRate:Vector3f=$Particles.MeshRotationRateBase) {isSetParameters=true}
   &ParticleState()
   &ScaleMeshSize(Scale Factor=&VectorFromFloat(Value=&FloatFromCurve(FloatCurve=DataInterface() {propertyValues {Curve {keys [
          {arriveTangent=1.5984, interpMode=RCIM_Cubic, leaveTangent=1.5984, tangentMode=RCTM_User, time=0, value=0.5}
          {arriveTangent=0.41, interpMode=RCIM_Cubic, leaveTangent=0.41, tangentMode=RCTM_User, time=1, value=1.25}
         ]}}})))
   &ScaleColor(
    Scale Alpha=&FloatFromCurve(FloatCurve=DataInterface() {propertyValues {Curve {keys [
         {arriveTangent=26.35, interpMode=RCIM_Cubic, leaveTangent=26.35, tangentMode=RCTM_User, time=0, value=0}
         <at 0.1 1 RCIM_Cubic>
         <at 1 0 RCIM_Cubic>
        ]}}})
    Scale RGB=&VectorFromCurve(VectorCurve=DataInterface() {propertyValues {
       XCurve {keys [<at 0 1 RCIM_Cubic>, {arriveTangent=-0.95, interpMode=RCIM_Cubic, leaveTangent=-0.95, time=0.5, value=0.05}, <at 1 0.05 RCIM_Cubic>]}
       YCurve {keys [<at 0 0.05 RCIM_Cubic>, {arriveTangent=0.95, interpMode=RCIM_Cubic, leaveTangent=0.95, time=0.5, value=0.6122}, <at 1 1 RCIM_Cubic>]}
       ZCurve {keys [<at 0 0.77 RCIM_Cubic>, {arriveTangent=-0.4762, interpMode=RCIM_Cubic, leaveTangent=-0.4762, time=0.5, value=1}, <at 1 0.2939 RCIM_Cubic>]}
      }})
    ScaleA=true
    ScaleRGB=true
   )
   /CascadeToNiagaraConverter/NiagaraScripts/ModuleScripts/UpdateMeshOrientationByEuler(RotationRate:Vector3f=$Particles.MeshRotationRate)
   &DynamicMaterialParameters(
    Index 0 Param 4=1
    Module.Param0WriteEnabled=false
    Module.Param1WriteEnabled=false
    Module.Param2WriteEnabled=false
    Module.Param3WriteEnabled=true
   ) {module=DynamicMaterialParameters001}
  ]
  rendererClasses [&NiagaraMeshRendererProperties]
  renderers [{rendererClass=&NiagaraMeshRendererProperties, rendererIndex=0}]
 }
}
summary {
 systemName=NS_DiscoBall
 emitters [
  {emitterName=DiscoBallSpawn, rendererClasses [&NiagaraMeshRendererProperties]}
  {bEnabled=false, emitterName=DiscoBallLoop, rendererClasses [&NiagaraMeshRendererProperties]}
  {emitterName=SparkleStationary, rendererClasses [&NiagaraSpriteRendererProperties]}
  {emitterName=SplineBeams, rendererClasses [&NiagaraMeshRendererProperties]}
  {emitterName=Light, rendererClasses [&NiagaraLightRendererProperties]}
  {emitterName=FlareStart, rendererClasses [&NiagaraSpriteRendererProperties, &NiagaraLightRendererProperties]}
  {emitterName=Corona, rendererClasses [&NiagaraSpriteRendererProperties]}
  {emitterName=CoronaSpread, rendererClasses [&NiagaraSpriteRendererProperties]}
 ]
 userVariables []
}
